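# Physics asset for the Shark Attack head cosmetic on the medium male player character.
# It holds two capsule bodies, a sideways capsule on the dyn_hoodBack bone and a small one on the head bone.
# A single constraint links the hood back to the head with limited swing and twist, bounding how far the hood can move relative to the head.
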PhysicsAsset /BRCosmetics/Characters/Player/Male/Medium/Heads/M_MED_Shark_Attack_Head/Meshes/Male_Medium_Shark_Attack_Head_Physics
bodies {
 dyn_hoodBack [{shapeType=Capsule, length=4.5, radius=3.5, center=<4.0863 0 0>, rotation=<rot 0 0 90>}]
 head [{shapeType=Capsule, length=1, radius=1}]
}
constraints [{bone1Name=dyn_hoodBack, bone2Name=head, swing1LimitDegrees=20, swing2LimitDegrees=20, twistLimitDegrees=25}]
